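# Secondary-motion AnimBlueprint for the female medium Hippie cosmetic body, copying its pose from the attached parent mesh.
# Thigh-versus-pelvis alphas, computed each update, drive additive offsets on skirt bones, except the right outer offset, which stays at full alpha.
# Chained AnimDynamics nodes then simulate bell-bottoms, sleeves, necklaces, amulet strings and skirt, using hand and spine limits on some.
# Final constraints position and rotate centre-front skirt bones from their left and right neighbours.
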AnimBlueprint /BRCosmetics/Characters/Player/Female/Medium/Bodies/F_MED_Hippie_01/Meshes/F_MED_Hippie_Body_AnimBP
target_skeleton=&/F_MED_Hippie_Body_Skeleton
graphs [
 AnimGraph {nodes [
   n1 = Get thigh_rot_l()
   n2 = "Transform (Modify) Bone - Bone: dyn_skirt_out_01_l"(
    ComponentPose=Local To Component(LocalPose=Copy Pose From Mesh() {bUseAttachedParent=true})
    Alpha=@n1
    Translation="2,0,0"
    Rotation="0,0,-15"
    Scale="1,1,1"
   ) {
    rotationMode=BMM_Additive
    rotationSpace=BCS_BoneSpace
    translationMode=BMM_Additive
    translationSpace=BCS_BoneSpace
    alphaBoolBlend {blendOption=Linear}
    boneToModify {boneName=dyn_skirt_out_01_l}
    rotation {roll=-15}
    translation {x=2}
   }
   n3 = "Transform (Modify) Bone - Bone: dyn_skirt_bk_01_l"(ComponentPose=@n2, Alpha=@n1, Translation="0,0,3", Rotation="60,0.000,0.000", Scale="1,1,1") {
    rotationMode=BMM_Additive
    rotationSpace=BCS_BoneSpace
    translationMode=BMM_Additive
    translationSpace=BCS_BoneSpace
    alphaBoolBlend {blendOption=Linear}
    boneToModify {boneName=dyn_skirt_bk_01_l}
    rotation {pitch=60}
    translation {z=3}
   }
   n4 = "Transform (Modify) Bone - Bone: dyn_skirt_out_01_r"(ComponentPose=@n3, Alpha=1, Translation="-2,0.000,0", Rotation="0,0.000,-15", Scale="1,1,1") {
    rotationMode=BMM_Additive
    rotationSpace=BCS_BoneSpace
    translationMode=BMM_Additive
    translationSpace=BCS_BoneSpace
    alphaBoolBlend {blendOption=Linear}
    boneToModify {boneName=dyn_skirt_out_01_r}
    rotation {roll=-15}
    translation {x=-2}
   }
   n5 = "Transform (Modify) Bone - Bone: dyn_skirt_bk_01_r"(ComponentPose=@n4, Alpha=Get thigh_rot_r(), Translation="0.000,0.000,-3", Rotation="60,0.000,0.000", Scale="1,1,1") {
    rotationMode=BMM_Additive
    rotationSpace=BCS_BoneSpace
    translationMode=BMM_Additive
    translationSpace=BCS_BoneSpace
    alphaBoolBlend {blendOption=Linear}
    boneToModify {boneName=dyn_skirt_bk_01_r}
    rotation {pitch=60}
    translation {z=-3}
   }
   n6 = Anim Dynamics - Bone: dyn_bellbottom_l(ComponentPose=@n5, Alpha=0.65, ExternalForce="0,0,0") {
    bGravityOverrideInSimSpace=true
    bLinearSpring=true
    bOverrideLinearDamping=true
    linearDampingOverride=0.98
    linearSpringConstant=125
    numSolverIterationsPostUpdate=4
    numSolverIterationsPreUpdate=16
    alphaBoolBlend {blendOption=Linear}
    boundBone {boneName=dyn_bellbottom_l}
    physicsBodyDefinitions [{
      sphereCollisionRadius=0
      boundBone {boneName=dyn_bellbottom_l}
      boxExtents=<25 25 5>
      constraintSetup {angularLimitsMax=<5 5 5>, angularLimitsMin=<-5 -5 -5>, angularTarget=<0 0 0>, linearAxesMax=<4 4 0>, linearAxesMin=<-4 -4 0>}
      localJointOffset=<0 0 -10>
     }]
   }
   n7 = Anim Dynamics - Bone: dyn_bellbottom_r(ComponentPose=@n6, Alpha=0.65, ExternalForce="0,0,0") {
    bGravityOverrideInSimSpace=true
    bLinearSpring=true
    bOverrideLinearDamping=true
    linearDampingOverride=0.98
    linearSpringConstant=125
    numSolverIterationsPostUpdate=4
    numSolverIterationsPreUpdate=16
    alphaBoolBlend {blendOption=Linear}
    boundBone {boneName=dyn_bellbottom_r}
    physicsBodyDefinitions [{
      sphereCollisionRadius=0
      boundBone {boneName=dyn_bellbottom_r}
      boxExtents=<25 25 5>
      constraintSetup {angularLimitsMax=<5 5 5>, angularLimitsMin=<-5 -5 -5>, angularTarget=<0 0 0>, linearAxesMax=<4 4 0>, linearAxesMin=<-4 -4 0>}
      localJointOffset=<0 0 -10>
     }]
   }
   n8 = Anim Dynamics - Bone: dyn_uppersleeve_l(ComponentPose=@n7, Alpha=0.1, ExternalForce="0,0,0") {
    bGravityOverrideInSimSpace=true
    numSolverIterationsPostUpdate=4
    numSolverIterationsPreUpdate=16
    alphaBoolBlend {blendOption=Linear}
    boundBone {boneName=dyn_uppersleeve_l}
    physicsBodyDefinitions [{
      sphereCollisionRadius=0
      boundBone {boneName=dyn_uppersleeve_l}
      boxExtents=<10 20 10>
      constraintSetup {angularLimitsMax=<15 15 15>, angularLimitsMin=<-15 -15 -15>, angularTarget=<0 0 0>}
      localJointOffset=<10 0 -5>
     }]
   }
   n9 = Anim Dynamics - Bone: dyn_uppersleeve_r(ComponentPose=@n8, Alpha=0.1, ExternalForce="0,0,0") {
    bGravityOverrideInSimSpace=true
    numSolverIterationsPostUpdate=4
    numSolverIterationsPreUpdate=16
    alphaBoolBlend {blendOption=Linear}
    boundBone {boneName=dyn_uppersleeve_r}
    physicsBodyDefinitions [{
      sphereCollisionRadius=0
      boundBone {boneName=dyn_uppersleeve_r}
      boxExtents=<10 20 10>
      constraintSetup {angularLimitsMax=<15 15 15>, angularLimitsMin=<-15 -15 -15>, angularTarget=<0 0 0>}
      localJointOffset=<-10 0 5>
     }]
   }
   n10 = Anim Dynamics - Bone: dyn_lowersleeve_l(ComponentPose=@n9, Alpha=1, ExternalForce="0,0,0") {
    angularDampingOverride=0.98
    angularSpringConstant=170
    bAngularSpring=true
    bGravityOverrideInSimSpace=true
    bLinearSpring=true
    bOverrideAngularDamping=true
    bOverrideLinearDamping=true
    bUsePlanarLimit=false
    bUseSphericalLimits=true
    gravityScale=0.125
    linearDampingOverride=0.98
    linearSpringConstant=170
    numSolverIterationsPostUpdate=4
    numSolverIterationsPreUpdate=16
    alphaBoolBlend {blendOption=Linear}
    boundBone {boneName=dyn_lowersleeve_l}
    physicsBodyDefinitions [{
      sphereCollisionRadius=0
      boundBone {boneName=dyn_lowersleeve_l}
      boxExtents=<10 20 20>
      constraintSetup {angularLimitsMax=<0 0 0>, angularLimitsMin=<0 0 0>, angularTarget=<0 0 0>, linearAxesMax=<0 3 3>, linearAxesMin=<0 -3 -3>}
      localJointOffset=<0 0 0>
     }]
    sphericalLimits [{limitRadius=5, limitType=Inner, drivingBone {boneName=hand_l}, sphereLocalOffset=<0 0 0>}]
   }
   n11 = Anim Dynamics - Bone: dyn_lowersleeve_r(ComponentPose=@n10, Alpha=1, ExternalForce="0,0,0") {
    angularDampingOverride=0.98
    angularSpringConstant=170
    bAngularSpring=true
    bGravityOverrideInSimSpace=true
    bLinearSpring=true
    bOverrideAngularDamping=true
    bOverrideLinearDamping=true
    bUseSphericalLimits=true
    gravityScale=0.125
    linearDampingOverride=0.98
    linearSpringConstant=170
    numSolverIterationsPostUpdate=4
    numSolverIterationsPreUpdate=16
    alphaBoolBlend {blendOption=Linear}
    boundBone {boneName=dyn_lowersleeve_r}
    physicsBodyDefinitions [{
      sphereCollisionRadius=0
      boundBone {boneName=dyn_lowersleeve_r}
      boxExtents=<10 20 20>
      constraintSetup {angularLimitsMax=<0 0 0>, angularLimitsMin=<0 0 0>, angularTarget=<0 0 0>, linearAxesMax=<0 5 5>, linearAxesMin=<0 -5 -5>}
      localJointOffset=<0 0 0>
     }]
    planarLimits [{drivingBone {boneName=hand_r}, planeTransform {location=<0 0 0>, rotation=<rot 0 0 0>}}]
    sphericalLimits [{limitRadius=5, limitType=Inner, drivingBone {boneName=hand_r}, sphereLocalOffset=<0 0 0>}]
   }
   n12 = Anim Dynamics - Bone: dyn_necklace_01(ComponentPose=@n11, Alpha=1, ExternalForce="0,0,0") {
    angularDampingOverride=0.99
    bGravityOverrideInSimSpace=true
    bOverrideAngularDamping=true
    bUseSphericalLimits=true
    numSolverIterationsPostUpdate=4
    numSolverIterationsPreUpdate=16
    alphaBoolBlend {blendOption=Linear}
    boundBone {boneName=dyn_necklace_01}
    physicsBodyDefinitions [{
      collisionType=CustomSphere
      sphereCollisionRadius=1
      boundBone {boneName=dyn_necklace_01}
      boxExtents=<30 30 40>
      constraintSetup {angularTargetAxis=AxisZ, twistAxis=AxisZ, angularLimitsMax=<40 60 0>, angularLimitsMin=<0 -60 0>, angularTarget=<0 0 0>}
      localJointOffset=<0 0 -5>
     }]
    sphericalLimits [{limitRadius=20, drivingBone {boneName=spine_04}, sphereLocalOffset=<-4 -6.3 0>}]
   }
   n13 = Anim Dynamics - Bone: dyn_necklace_02(ComponentPose=@n12, Alpha=1, ExternalForce="0,0,0") {
    angularDampingOverride=0.99
    bGravityOverrideInSimSpace=true
    bOverrideAngularDamping=true
    bUseSphericalLimits=true
    numSolverIterationsPostUpdate=4
    numSolverIterationsPreUpdate=16
    alphaBoolBlend {blendOption=Linear}
    boundBone {boneName=dyn_necklace_02}
    physicsBodyDefinitions [{
      collisionType=CustomSphere
      sphereCollisionRadius=1
      boundBone {boneName=dyn_necklace_02}
      boxExtents=<30 30 40>
      constraintSetup {angularTargetAxis=AxisZ, twistAxis=AxisZ, angularLimitsMax=<40 60 0>, angularLimitsMin=<0 -60 0>, angularTarget=<0 0 0>}
      localJointOffset=<0 0 -5>
     }]
    sphericalLimits [{limitRadius=20, drivingBone {boneName=spine_04}, sphereLocalOffset=<-4 -6.5 0>}]
   }
   n14 = Anim Dynamics - Bone: dyn_amulet_01(ComponentPose=@n13, Alpha=1, ExternalForce="0,0,0") {
    angularDampingOverride=0.99
    bGravityOverrideInSimSpace=true
    bOverrideAngularDamping=true
    numSolverIterationsPostUpdate=4
    numSolverIterationsPreUpdate=16
    alphaBoolBlend {blendOption=Linear}
    boundBone {boneName=dyn_amulet_01}
    physicsBodyDefinitions [{
      sphereCollisionRadius=0
      boundBone {boneName=dyn_amulet_01}
      boxExtents=<35 35 10>
      constraintSetup {twistAxis=AxisZ, angularLimitsMax=<40 45 0>, angularLimitsMin=<-8 -45 0>, angularTarget=<0 0 0>}
      localJointOffset=<0 0 -5>
     }]
   }
   n15 = Anim Dynamics - Bone: dyn_amulet_string_l_01(ComponentPose=@n14, Alpha=1, ExternalForce="0,0,0") {
    angularDampingOverride=0.99
    bGravityOverrideInSimSpace=true
    bOverrideAngularDamping=true
    numSolverIterationsPostUpdate=4
    numSolverIterationsPreUpdate=16
    alphaBoolBlend {blendOption=Linear}
    boundBone {boneName=dyn_amulet_string_l_01}
    physicsBodyDefinitions [{
      sphereCollisionRadius=0
      boundBone {boneName=dyn_amulet_string_l_01}
      boxExtents=<20 20 40>
      constraintSetup {angularTargetAxis=AxisZ, twistAxis=AxisZ, angularLimitsMax=<15 45 0>, angularLimitsMin=<-10 -45 0>, angularTarget=<0 0 0>}
      localJointOffset=<0 0 -10>
     }]
   }
   n16 = Anim Dynamics - Bone: dyn_amulet_string_l_02(ComponentPose=@n15, Alpha=1, ExternalForce="0,0,0") {
    angularDampingOverride=0.98
    bGravityOverrideInSimSpace=true
    bOverrideAngularDamping=true
    numSolverIterationsPostUpdate=4
    numSolverIterationsPreUpdate=16
    alphaBoolBlend {blendOption=Linear}
    boundBone {boneName=dyn_amulet_string_l_02}
    physicsBodyDefinitions [{
      sphereCollisionRadius=0
      boundBone {boneName=dyn_amulet_string_l_02}
      boxExtents=<15 15 30>
      constraintSetup {angularTargetAxis=AxisZ, twistAxis=AxisZ, angularLimitsMax=<10 45 0>, angularLimitsMin=<-10 -45 0>, angularTarget=<0 0 0>}
      localJointOffset=<0 0 -3>
     }]
   }
   n17 = Anim Dynamics - Bone: dyn_amulet_string_c_01(ComponentPose=@n16, Alpha=1, ExternalForce="0,0,0") {
    angularDampingOverride=0.99
    bGravityOverrideInSimSpace=true
    bOverrideAngularDamping=true
    numSolverIterationsPostUpdate=4
    numSolverIterationsPreUpdate=16
    alphaBoolBlend {blendOption=Linear}
    boundBone {boneName=dyn_amulet_string_c_01}
    physicsBodyDefinitions [{
      sphereCollisionRadius=0
      boundBone {boneName=dyn_amulet_string_c_01}
      boxExtents=<25 18 38>
      constraintSetup {angularTargetAxis=AxisZ, twistAxis=AxisZ, angularLimitsMax=<15 45 0>, angularLimitsMin=<-10 -45 0>, angularTarget=<0 0 0>}
      localJointOffset=<0 0 -8>
     }]
   }
   n18 = Anim Dynamics - Bone: dyn_amulet_string_c_02(ComponentPose=@n17, Alpha=1, ExternalForce="0,0,0") {
    angularDampingOverride=0.98
    bGravityOverrideInSimSpace=true
    bOverrideAngularDamping=true
    numSolverIterationsPostUpdate=4
    numSolverIterationsPreUpdate=16
    alphaBoolBlend {blendOption=Linear}
    boundBone {boneName=dyn_amulet_string_c_02}
    physicsBodyDefinitions [{
      sphereCollisionRadius=0
      boundBone {boneName=dyn_amulet_string_c_02}
      boxExtents=<18 16 32>
      constraintSetup {angularTargetAxis=AxisZ, twistAxis=AxisZ, angularLimitsMax=<10 45 0>, angularLimitsMin=<-10 -45 0>, angularTarget=<0 0 0>}
      localJointOffset=<0 0 -4>
     }]
   }
   n19 = Anim Dynamics - Bone: dyn_amulet_string_r_01(ComponentPose=@n18, Alpha=1, ExternalForce="0,0,0") {
    angularDampingOverride=0.99
    bGravityOverrideInSimSpace=true
    bOverrideAngularDamping=true
    numSolverIterationsPostUpdate=4
    numSolverIterationsPreUpdate=16
    alphaBoolBlend {blendOption=Linear}
    boundBone {boneName=dyn_amulet_string_r_01}
    physicsBodyDefinitions [{
      sphereCollisionRadius=0
      boundBone {boneName=dyn_amulet_string_r_01}
      boxExtents=<16 22 36>
      constraintSetup {angularTargetAxis=AxisZ, twistAxis=AxisZ, angularLimitsMax=<15 45 0>, angularLimitsMin=<-10 -45 0>, angularTarget=<0 0 0>}
      localJointOffset=<0 0 -5>
     }]
   }
   n20 = Anim Dynamics - Bone: dyn_amulet_string_r_02(ComponentPose=@n19, Alpha=1, ExternalForce="0,0,0") {
    angularDampingOverride=0.98
    bGravityOverrideInSimSpace=true
    bOverrideAngularDamping=true
    numSolverIterationsPostUpdate=4
    numSolverIterationsPreUpdate=16
    alphaBoolBlend {blendOption=Linear}
    boundBone {boneName=dyn_amulet_string_r_02}
    physicsBodyDefinitions [{
      sphereCollisionRadius=0
      boundBone {boneName=dyn_amulet_string_r_02}
      boxExtents=<18 12 29>
      constraintSetup {angularTargetAxis=AxisZ, twistAxis=AxisZ, angularLimitsMax=<10 45 0>, angularLimitsMin=<-10 -45 0>, angularTarget=<0 0 0>}
      localJointOffset=<0 0 -4>
     }]
   }
   n21 = Anim Dynamics - Bone: dyn_skirt_out_01_l!(ComponentPose=@n20, ExternalForce="0,0,0") -> (Pose) {
    angularDampingOverride=0.99
    angularSpringConstant=15000
    bAngularSpring=true
    bGravityOverrideInSimSpace=true
    bOverrideAngularDamping=true
    numSolverIterationsPostUpdate=4
    numSolverIterationsPreUpdate=16
    alphaBoolBlend {blendOption=Linear}
    boundBone {boneName=dyn_skirt_out_01_l}
    physicsBodyDefinitions [{
      sphereCollisionRadius=0
      boundBone {boneName=dyn_skirt_out_01_l}
      boxExtents=<50 50 30>
      constraintSetup {angularTargetAxis=AxisZ, twistAxis=AxisZ, angularLimitsMax=<5 0 5>, angularLimitsMin=<-20 -15 -5>, angularTarget=<0 0 1>}
      localJointOffset=<0 0 -5>
     }]
   }
   n22 = Anim Dynamics - Bone: dyn_skirt_out_02_l!(ComponentPose=@n21, ExternalForce="0,0,0") -> (Pose) {
    angularDampingOverride=0.99
    angularSpringConstant=15000
    bAngularSpring=true
    bGravityOverrideInSimSpace=true
    bOverrideAngularDamping=true
    numSolverIterationsPostUpdate=4
    numSolverIterationsPreUpdate=16
    alphaBoolBlend {blendOption=Linear}
    boundBone {boneName=dyn_skirt_out_02_l}
    physicsBodyDefinitions [{
      sphereCollisionRadius=0
      boundBone {boneName=dyn_skirt_out_02_l}
      boxExtents=<50 50 30>
      constraintSetup {angularTargetAxis=AxisZ, twistAxis=AxisZ, angularLimitsMax=<20 0 5>, angularLimitsMin=<-20 -15 -5>, angularTarget=<0 0 1>}
      localJointOffset=<0 0 -5>
     }]
   }
   n23 = Anim Dynamics - Bone: dyn_skirt_fr_01_l!(ComponentPose=@n22, ExternalForce="0,0,0") -> (Pose) {
    angularDampingOverride=0.99
    angularSpringConstant=15000
    bAngularSpring=true
    bGravityOverrideInSimSpace=true
    bOverrideAngularDamping=true
    numSolverIterationsPostUpdate=4
    numSolverIterationsPreUpdate=16
    alphaBoolBlend {blendOption=Linear}
    boundBone {boneName=dyn_skirt_fr_01_l}
    physicsBodyDefinitions [{
      sphereCollisionRadius=0
      boundBone {boneName=dyn_skirt_fr_01_l}
      boxExtents=<50 50 30>
      constraintSetup {angularTargetAxis=AxisZ, twistAxis=AxisZ, angularLimitsMax=<0 0 0>, angularLimitsMin=<0 -15 0>, angularTarget=<0 0 1>}
      localJointOffset=<0 0 -5>
     }]
   }
   n24 = Anim Dynamics - Bone: dyn_skirt_fr_02_l!(ComponentPose=@n23, ExternalForce="0,0,0") -> (Pose) {
    angularDampingOverride=0.99
    angularSpringConstant=15000
    bAngularSpring=true
    bGravityOverrideInSimSpace=true
    bOverrideAngularDamping=true
    numSolverIterationsPostUpdate=4
    numSolverIterationsPreUpdate=16
    alphaBoolBlend {blendOption=Linear}
    boundBone {boneName=dyn_skirt_fr_02_l}
    physicsBodyDefinitions [{
      sphereCollisionRadius=0
      boundBone {boneName=dyn_skirt_fr_02_l}
      boxExtents=<50 50 30>
      constraintSetup {angularTargetAxis=AxisZ, twistAxis=AxisZ, angularLimitsMax=<0 0 5>, angularLimitsMin=<-5 -15 -5>, angularTarget=<0 0 1>}
      localJointOffset=<0 0 -5>
     }]
   }
   n25 = Anim Dynamics - Bone: dyn_skirt_bk_01_l!(ComponentPose=@n24, ExternalForce="0,0,0") -> (Pose) {
    angularDampingOverride=0.99
    angularSpringConstant=15000
    bAngularSpring=true
    bGravityOverrideInSimSpace=true
    bOverrideAngularDamping=true
    numSolverIterationsPostUpdate=4
    numSolverIterationsPreUpdate=16
    alphaBoolBlend {blendOption=Linear}
    boundBone {boneName=dyn_skirt_bk_01_l}
    physicsBodyDefinitions [{
      sphereCollisionRadius=0
      boundBone {boneName=dyn_skirt_bk_01_l}
      boxExtents=<50 50 30>
      constraintSetup {angularTargetAxis=AxisZ, twistAxis=AxisZ, angularLimitsMax=<20 0 5>, angularLimitsMin=<-20 -90 -5>, angularTarget=<0 0 1>}
      localJointOffset=<0 0 -5>
     }]
   }
   n26 = Anim Dynamics - Bone: dyn_skirt_bk_02_l!(ComponentPose=@n25, ExternalForce="0,0,0") -> (Pose) {
    angularDampingOverride=0.99
    angularSpringConstant=15000
    bAngularSpring=true
    bGravityOverrideInSimSpace=true
    bOverrideAngularDamping=true
    numSolverIterationsPostUpdate=4
    numSolverIterationsPreUpdate=16
    alphaBoolBlend {blendOption=Linear}
    boundBone {boneName=dyn_skirt_bk_02_l}
    physicsBodyDefinitions [{
      sphereCollisionRadius=0
      boundBone {boneName=dyn_skirt_bk_02_l}
      boxExtents=<50 50 30>
      constraintSetup {angularTargetAxis=AxisZ, twistAxis=AxisZ, angularLimitsMax=<20 0 5>, angularLimitsMin=<-20 -90 -5>, angularTarget=<0 0 1>}
      localJointOffset=<0 0 -5>
     }]
   }
   n27 = Anim Dynamics - Bone: dyn_skirt_frdiag_01_l!(ComponentPose=@n26, ExternalForce="0,0,0") -> (Pose) {
    angularDampingOverride=0.99
    angularSpringConstant=15000
    bAngularSpring=true
    bGravityOverrideInSimSpace=true
    bOverrideAngularDamping=true
    numSolverIterationsPostUpdate=4
    numSolverIterationsPreUpdate=16
    alphaBoolBlend {blendOption=Linear}
    boundBone {boneName=dyn_skirt_frdiag_01_l}
    physicsBodyDefinitions [{
      sphereCollisionRadius=0
      boundBone {boneName=dyn_skirt_frdiag_01_l}
      boxExtents=<50 50 30>
      constraintSetup {angularTargetAxis=AxisZ, twistAxis=AxisZ, angularLimitsMax=<10 0 0>, angularLimitsMin=<-10 -15 0>, angularTarget=<0 0 1>}
      localJointOffset=<0 0 -5>
     }]
   }
   n28 = Anim Dynamics - Bone: dyn_skirt_frdiag_02_l!(ComponentPose=@n27, ExternalForce="0,0,0") -> (Pose) {
    angularDampingOverride=0.99
    angularSpringConstant=15000
    bAngularSpring=true
    bGravityOverrideInSimSpace=true
    bOverrideAngularDamping=true
    numSolverIterationsPostUpdate=4
    numSolverIterationsPreUpdate=16
    alphaBoolBlend {blendOption=Linear}
    boundBone {boneName=dyn_skirt_frdiag_02_l}
    physicsBodyDefinitions [{
      sphereCollisionRadius=0
      boundBone {boneName=dyn_skirt_frdiag_02_l}
      boxExtents=<50 50 30>
      constraintSetup {angularTargetAxis=AxisZ, twistAxis=AxisZ, angularLimitsMax=<10 0 5>, angularLimitsMin=<-10 -15 -5>, angularTarget=<0 0 1>}
      localJointOffset=<0 0 -5>
     }]
   }
   n29 = Anim Dynamics - Bone: dyn_skirt_out_01_r!(ComponentPose=@n28, ExternalForce="0,0,0") -> (Pose) {
    angularDampingOverride=0.99
    angularSpringConstant=15000
    bAngularSpring=true
    bGravityOverrideInSimSpace=true
    bOverrideAngularDamping=true
    numSolverIterationsPostUpdate=4
    numSolverIterationsPreUpdate=16
    alphaBoolBlend {blendOption=Linear}
    boundBone {boneName=dyn_skirt_out_01_r}
    physicsBodyDefinitions [{
      sphereCollisionRadius=0
      boundBone {boneName=dyn_skirt_out_01_r}
      boxExtents=<50 50 30>
      constraintSetup {angularTargetAxis=AxisZ, twistAxis=AxisZ, angularLimitsMax=<5 0 5>, angularLimitsMin=<-20 -15 -5>, angularTarget=<0 0 1>}
      localJointOffset=<0 0 5>
     }]
   }
   n30 = Anim Dynamics - Bone: dyn_skirt_out_02_r!(ComponentPose=@n29, ExternalForce="0,0,0") -> (Pose) {
    angularDampingOverride=0.99
    angularSpringConstant=15000
    bAngularSpring=true
    bGravityOverrideInSimSpace=true
    bOverrideAngularDamping=true
    numSolverIterationsPostUpdate=4
    numSolverIterationsPreUpdate=16
    alphaBoolBlend {blendOption=Linear}
    boundBone {boneName=dyn_skirt_out_02_r}
    physicsBodyDefinitions [{
      sphereCollisionRadius=0
      boundBone {boneName=dyn_skirt_out_02_r}
      boxExtents=<50 50 30>
      constraintSetup {angularTargetAxis=AxisZ, twistAxis=AxisZ, angularLimitsMax=<20 0 5>, angularLimitsMin=<-20 -15 -5>, angularTarget=<0 0 1>}
      localJointOffset=<0 0 5>
     }]
   }
   n31 = Anim Dynamics - Bone: dyn_skirt_fr_01_r!(ComponentPose=@n30, ExternalForce="0,0,0") -> (Pose) {
    angularDampingOverride=0.99
    angularSpringConstant=15000
    bAngularSpring=true
    bGravityOverrideInSimSpace=true
    bOverrideAngularDamping=true
    numSolverIterationsPostUpdate=4
    numSolverIterationsPreUpdate=16
    alphaBoolBlend {blendOption=Linear}
    boundBone {boneName=dyn_skirt_fr_01_r}
    physicsBodyDefinitions [{
      sphereCollisionRadius=0
      boundBone {boneName=dyn_skirt_fr_01_r}
      boxExtents=<50 50 30>
      constraintSetup {angularTargetAxis=AxisZ, twistAxis=AxisZ, angularLimitsMax=<0 0 0>, angularLimitsMin=<0 -15 0>, angularTarget=<0 0 1>}
      localJointOffset=<0 0 5>
     }]
   }
   n32 = Anim Dynamics - Bone: dyn_skirt_fr_02_r!(ComponentPose=@n31, ExternalForce="0,0,0") -> (Pose) {
    angularDampingOverride=0.99
    angularSpringConstant=15000
    bAngularSpring=true
    bGravityOverrideInSimSpace=true
    bOverrideAngularDamping=true
    numSolverIterationsPostUpdate=4
    numSolverIterationsPreUpdate=16
    alphaBoolBlend {blendOption=Linear}
    boundBone {boneName=dyn_skirt_fr_02_r}
    physicsBodyDefinitions [{
      sphereCollisionRadius=0
      boundBone {boneName=dyn_skirt_fr_02_r}
      boxExtents=<50 50 30>
      constraintSetup {angularTargetAxis=AxisZ, twistAxis=AxisZ, angularLimitsMax=<0 0 5>, angularLimitsMin=<-5 -15 -5>, angularTarget=<0 0 1>}
      localJointOffset=<0 0 5>
     }]
   }
   n33 = Anim Dynamics - Bone: dyn_skirt_bk_01_r!(ComponentPose=@n32, ExternalForce="0,0,0") -> (Pose) {
    angularDampingOverride=0.99
    angularSpringConstant=15000
    bAngularSpring=true
    bGravityOverrideInSimSpace=true
    bOverrideAngularDamping=true
    numSolverIterationsPostUpdate=4
    numSolverIterationsPreUpdate=16
    alphaBoolBlend {blendOption=Linear}
    boundBone {boneName=dyn_skirt_bk_01_r}
    physicsBodyDefinitions [{
      sphereCollisionRadius=0
      boundBone {boneName=dyn_skirt_bk_01_r}
      boxExtents=<50 50 30>
      constraintSetup {angularTargetAxis=AxisZ, twistAxis=AxisZ, angularLimitsMax=<20 0 5>, angularLimitsMin=<-20 -90 -5>, angularTarget=<0 0 1>}
      localJointOffset=<0 0 5>
     }]
   }
   n34 = Anim Dynamics - Bone: dyn_skirt_bk_02_r!(ComponentPose=@n33, ExternalForce="0,0,0") -> (Pose) {
    angularDampingOverride=0.99
    angularSpringConstant=15000
    bAngularSpring=true
    bGravityOverrideInSimSpace=true
    bOverrideAngularDamping=true
    numSolverIterationsPostUpdate=4
    numSolverIterationsPreUpdate=16
    alphaBoolBlend {blendOption=Linear}
    boundBone {boneName=dyn_skirt_bk_02_r}
    physicsBodyDefinitions [{
      sphereCollisionRadius=0
      boundBone {boneName=dyn_skirt_bk_02_r}
      boxExtents=<50 50 30>
      constraintSetup {angularTargetAxis=AxisZ, twistAxis=AxisZ, angularLimitsMax=<20 0 5>, angularLimitsMin=<-20 -90 -5>, angularTarget=<0 0 1>}
      localJointOffset=<0 0 5>
     }]
   }
   n35 = Anim Dynamics - Bone: dyn_skirt_frdiag_01_r!(ComponentPose=@n34, ExternalForce="0,0,0") -> (Pose) {
    angularDampingOverride=0.99
    angularSpringConstant=15000
    bAngularSpring=true
    bGravityOverrideInSimSpace=true
    bOverrideAngularDamping=true
    numSolverIterationsPostUpdate=4
    numSolverIterationsPreUpdate=16
    alphaBoolBlend {blendOption=Linear}
    boundBone {boneName=dyn_skirt_frdiag_01_r}
    physicsBodyDefinitions [{
      sphereCollisionRadius=0
      boundBone {boneName=dyn_skirt_frdiag_01_r}
      boxExtents=<50 50 30>
      constraintSetup {angularTargetAxis=AxisZ, twistAxis=AxisZ, angularLimitsMax=<10 0 0>, angularLimitsMin=<-10 -15 0>, angularTarget=<0 0 1>}
      localJointOffset=<0 0 5>
     }]
   }
   n36 = Anim Dynamics - Bone: dyn_skirt_frdiag_02_r!(ComponentPose=@n35, ExternalForce="0,0,0") -> (Pose) {
    angularDampingOverride=0.99
    angularSpringConstant=15000
    bAngularSpring=true
    bGravityOverrideInSimSpace=true
    bOverrideAngularDamping=true
    numSolverIterationsPostUpdate=4
    numSolverIterationsPreUpdate=16
    alphaBoolBlend {blendOption=Linear}
    boundBone {boneName=dyn_skirt_frdiag_02_r}
    physicsBodyDefinitions [{
      sphereCollisionRadius=0
      boundBone {boneName=dyn_skirt_frdiag_02_r}
      boxExtents=<50 50 30>
      constraintSetup {angularTargetAxis=AxisZ, twistAxis=AxisZ, angularLimitsMax=<10 0 5>, angularLimitsMin=<-10 -15 -5>, angularTarget=<0 0 1>}
      localJointOffset=<0 0 5>
     }]
   }
   n37 = "Constraint - dyn_skirt_fr_01 (TTRR)"!(ComponentPose=@n36, Alpha=1, ConstraintWeights_0=1, ConstraintWeights_1:float=1, ConstraintWeights_2:float=1, ConstraintWeights_3:float=1) -> (Pose) {
    alphaBoolBlend {blendOption=Linear}
    boneToModify {boneName=dyn_skirt_fr_01}
    constraintSetup [
     {transformType=Translation, targetBone {boneName=dyn_skirt_fr_01_l}}
     {transformType=Translation, targetBone {boneName=dyn_skirt_fr_01_r}}
     {transformType=Rotation, targetBone {boneName=dyn_skirt_fr_01_l}}
     {transformType=Rotation, targetBone {boneName=dyn_skirt_fr_01_r}}
    ]
    constraintWeights [1,1,1,1]
   }
   n38 = "Constraint - dyn_skirt_fr_02 (TTRR)"!(ComponentPose=@n37, Alpha=1, ConstraintWeights_0=1, ConstraintWeights_1:float=1, ConstraintWeights_2:float=1, ConstraintWeights_3:float=1) -> (Pose) {
    alphaBoolBlend {blendOption=Linear}
    boneToModify {boneName=dyn_skirt_fr_02}
    constraintSetup [
     {transformType=Translation, targetBone {boneName=dyn_skirt_fr_02_l}}
     {transformType=Translation, targetBone {boneName=dyn_skirt_fr_02_r}}
     {transformType=Rotation, targetBone {boneName=dyn_skirt_fr_02_l}}
     {transformType=Rotation, targetBone {boneName=dyn_skirt_fr_02_r}}
    ]
    constraintWeights [1,1,1,1]
   }
   Output Pose(Result=Component To Local(ComponentPose=@n38)) {layerGroup=None}
   Sequence Player 'Crouch'() {sequence=&/Crouch}
  ]}
 EventGraph {nodes [
   Event BlueprintUpdateAnimation()
   n1 = Get Skel_Mesh()
   > Set thigh_rot_l(thigh_rot_l=GetAlphaFromReferenceTransformDelta(SkeletalMeshComponent=@n1, BoneName=thigh_l, BaseName=pelvis, InputRange=(X=-20,Y=-100), TwistAxis="1,0,0"))
   > Set thigh_rot_r(thigh_rot_r=GetAlphaFromReferenceTransformDelta(SkeletalMeshComponent=@n1, BoneName=thigh_r, BaseName=pelvis, InputRange=(X=-20,Y=-100), TwistAxis="1,0,0"))
   Event BlueprintInitializeAnimation()
   > Set Skel_Mesh(Skel_Mesh=GetOwnerSkeletalMeshComponent())
  ]}
]
variables [
 Skel_Mesh {type=SkeletalMeshComponent}
 thigh_rot_l {type=number}
 thigh_rot_r {type=number}
]
