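# Post-process-style head AnimBP for the female Hippie 01 cosmetic, adding secondary motion to hair and braids.
# Copies the attached parent's pose, then chains Anim Dynamics through front braids and hair, some using spine/neck spherical limits.
# Weighted four-target constraints then drive three braid bones from head and hair bones, before Anim Dynamics swings the braid ends.
# Three dynamics nodes are disconnected from output, and the event graph is empty stubs.
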AnimBlueprint /BRCosmetics/Characters/Player/Female/Medium/Bodies/F_MED_Hippie_01/Meshes/Parts/F_MED_Hippie_Head_AnimBP
target_skeleton=&/F_MED_Hippie_Head_Skeleton
graphs [
 AnimGraph {nodes [
   n1 = Anim Dynamics - Bone: dyn_braids_02!(ComponentPose, ExternalForce="0,0,0") -> (Pose) {
    alpha=0.5
    angularDampingOverride=0.99
    bGravityOverrideInSimSpace=true
    lODThreshold=1
    numSolverIterationsPostUpdate=4
    numSolverIterationsPreUpdate=16
    alphaBoolBlend {blendOption=Linear}
    boundBone {boneName=dyn_braids_02}
    physicsBodyDefinitions [{
      sphereCollisionRadius=0
      boundBone {boneName=dyn_braids_02}
      boxExtents=<40 40 30>
      constraintSetup {angularTargetAxis=AxisZ, twistAxis=AxisZ, angularLimitsMax=<25 45 0>, angularLimitsMin=<0 -45 0>, angularTarget=<0 0 0>}
      localJointOffset=<0 0 -8>
     }]
   }
   Anim Dynamics - Bone: dyn_braids_03!(ComponentPose=@n1, ExternalForce="0,0,0") -> (Pose) {
    alpha=0.5
    angularDampingOverride=0.99
    bGravityOverrideInSimSpace=true
    lODThreshold=1
    numSolverIterationsPostUpdate=4
    numSolverIterationsPreUpdate=16
    alphaBoolBlend {blendOption=Linear}
    boundBone {boneName=dyn_braids_03}
    physicsBodyDefinitions [{
      sphereCollisionRadius=0
      boundBone {boneName=dyn_braids_03}
      boxExtents=<40 40 30>
      constraintSetup {angularTargetAxis=AxisZ, twistAxis=AxisZ, angularLimitsMax=<25 45 0>, angularLimitsMin=<0 -45 0>, angularTarget=<0 0 0>}
      localJointOffset=<0 0 -8>
     }]
   }
   Anim Dynamics - Bone: dyn_braid_front_03_b!(ComponentPose, ExternalForce="0,0,0") -> (Pose) {
    alpha=0.3
    angularDampingOverride=0.99
    angularSpringConstant=800
    bAngularSpring=true
    bGravityOverrideInSimSpace=true
    lODThreshold=1
    numSolverIterationsPostUpdate=4
    numSolverIterationsPreUpdate=16
    alphaBoolBlend {blendOption=Linear}
    boundBone {boneName=dyn_braid_front_03_b}
    physicsBodyDefinitions [{
      sphereCollisionRadius=0
      boundBone {boneName=dyn_braid_front_03_b}
      boxExtents=<40 40 30>
      constraintSetup {angularTargetAxis=AxisY, twistAxis=AxisZ, angularLimitsMax=<25 45 0>, angularLimitsMin=<0 -45 0>, angularTarget=<0 1 0>}
      localJointOffset=<0 0 -15>
     }]
   }
   n2 = Anim Dynamics - Bone: dyn_braid_front_01!(ComponentPose=Local To Component(LocalPose=Copy Pose From Mesh() {bUseAttachedParent=true}), ExternalForce="0,0,0") -> (Pose) {
    alpha=0.3
    angularDampingOverride=0.99
    angularSpringConstant=800
    bAngularSpring=true
    bGravityOverrideInSimSpace=true
    lODThreshold=1
    numSolverIterationsPostUpdate=4
    numSolverIterationsPreUpdate=16
    alphaBoolBlend {blendOption=Linear}
    boundBone {boneName=dyn_braid_front_01}
    physicsBodyDefinitions [{
      sphereCollisionRadius=0
      boundBone {boneName=dyn_braid_front_01}
      boxExtents=<40 40 30>
      constraintSetup {angularTargetAxis=AxisY, twistAxis=AxisZ, angularLimitsMax=<25 45 0>, angularLimitsMin=<0 -45 0>, angularTarget=<0 1 0>}
      localJointOffset=<0 0 -15>
     }]
   }
   n3 = Anim Dynamics - Bone: dyn_braid_front_03!(ComponentPose=@n2, ExternalForce="0,0,0") -> (Pose) {
    alpha=0.3
    angularDampingOverride=0.99
    angularSpringConstant=800
    bAngularSpring=true
    bGravityOverrideInSimSpace=true
    lODThreshold=1
    numSolverIterationsPostUpdate=4
    numSolverIterationsPreUpdate=16
    alphaBoolBlend {blendOption=Linear}
    boundBone {boneName=dyn_braid_front_03}
    physicsBodyDefinitions [{
      sphereCollisionRadius=0
      boundBone {boneName=dyn_braid_front_03}
      boxExtents=<40 40 30>
      constraintSetup {angularTargetAxis=AxisY, twistAxis=AxisZ, angularLimitsMax=<25 45 0>, angularLimitsMin=<0 -45 0>, angularTarget=<0 1 0>}
      localJointOffset=<0 0 -15>
     }]
   }
   n4 = Anim Dynamics - Bone: dyn_braid_front_03_a!(ComponentPose=@n3, ExternalForce="0,0,0") -> (Pose) {
    alpha=0.3
    angularDampingOverride=0.99
    angularSpringConstant=800
    bAngularSpring=true
    bGravityOverrideInSimSpace=true
    lODThreshold=1
    numSolverIterationsPostUpdate=4
    numSolverIterationsPreUpdate=16
    alphaBoolBlend {blendOption=Linear}
    boundBone {boneName=dyn_braid_front_03_a}
    physicsBodyDefinitions [{
      sphereCollisionRadius=0
      boundBone {boneName=dyn_braid_front_03_a}
      boxExtents=<40 40 30>
      constraintSetup {angularTargetAxis=AxisY, twistAxis=AxisZ, angularLimitsMax=<25 45 0>, angularLimitsMin=<0 -45 0>, angularTarget=<0 1 0>}
      localJointOffset=<0 0 -15>
     }]
   }
   n5 = Anim Dynamics - Bone: dyn_braid_front_03_b_01!(ComponentPose=@n4, ExternalForce="0,0,0") -> (Pose) {
    alpha=0.3
    angularDampingOverride=0.99
    angularSpringConstant=800
    bAngularSpring=true
    bGravityOverrideInSimSpace=true
    lODThreshold=1
    numSolverIterationsPostUpdate=4
    numSolverIterationsPreUpdate=16
    alphaBoolBlend {blendOption=Linear}
    boundBone {boneName=dyn_braid_front_03_b_01}
    physicsBodyDefinitions [{
      sphereCollisionRadius=0
      boundBone {boneName=dyn_braid_front_03_b_01}
      boxExtents=<40 40 30>
      constraintSetup {angularTargetAxis=AxisY, twistAxis=AxisZ, angularLimitsMax=<25 45 0>, angularLimitsMin=<0 -45 0>, angularTarget=<0 1 0>}
      localJointOffset=<0 0 -15>
     }]
   }
   n6 = Anim Dynamics - Bone: dyn_Hippie_hair_01(ComponentPose=@n5, Alpha=1, ExternalForce="0,0,0") {
    angularDampingOverride=0.98
    angularSpringConstant=1500
    bAngularSpring=true
    bGravityOverrideInSimSpace=true
    bOverrideAngularDamping=true
    numSolverIterationsPostUpdate=4
    numSolverIterationsPreUpdate=16
    alphaBoolBlend {blendOption=Linear}
    boundBone {boneName=dyn_Hippie_hair_01}
    physicsBodyDefinitions [{
      collisionType=CustomSphere
      sphereCollisionRadius=1
      boundBone {boneName=dyn_Hippie_hair_01}
      boxExtents=<20 45 45>
      constraintSetup {angularTargetAxis=AxisZ, angularLimitsMax=<5 20 0>, angularLimitsMin=<-5 -20 -20>, angularTarget=<0 0 1>}
      localJointOffset=<2 0 0>
     }]
   }
   n7 = Anim Dynamics - Bone: dyn_Hippie_hair_02(ComponentPose=@n6, Alpha=1, ExternalForce="0,0,0") {
    angularDampingOverride=0.98
    angularSpringConstant=1500
    bAngularSpring=true
    bGravityOverrideInSimSpace=true
    bOverrideAngularDamping=true
    numSolverIterationsPostUpdate=4
    numSolverIterationsPreUpdate=16
    alphaBoolBlend {blendOption=Linear}
    boundBone {boneName=dyn_Hippie_hair_02}
    physicsBodyDefinitions [{
      collisionType=CustomSphere
      sphereCollisionRadius=1
      boundBone {boneName=dyn_Hippie_hair_02}
      boxExtents=<20 45 45>
      constraintSetup {angularTargetAxis=AxisZ, angularLimitsMax=<5 20 0>, angularLimitsMin=<-5 -20 -20>, angularTarget=<0 0 1>}
      localJointOffset=<2 0 0>
     }]
   }
   n8 = Anim Dynamics - Bone: dyn_Hippie_hair_03(ComponentPose=@n7, Alpha=1, ExternalForce="0,0,0") {
    angularDampingOverride=0.98
    angularSpringConstant=100
    bAngularSpring=true
    bGravityOverrideInSimSpace=true
    bOverrideAngularDamping=true
    numSolverIterationsPostUpdate=4
    numSolverIterationsPreUpdate=16
    alphaBoolBlend {blendOption=Linear}
    boundBone {boneName=dyn_Hippie_hair_03}
    physicsBodyDefinitions [{
      collisionType=CustomSphere
      sphereCollisionRadius=1
      boundBone {boneName=dyn_Hippie_hair_03}
      boxExtents=<20 30 30>
      constraintSetup {angularTargetAxis=AxisZ, angularLimitsMax=<5 20 -5>, angularLimitsMin=<-5 -20 -60>, angularTarget=<0 0 1>}
      localJointOffset=<2 0 0>
     }]
    sphericalLimits [{limitRadius=12, drivingBone {boneName=spine_05}, sphereLocalOffset=<0 0 0>}]
   }
   n9 = Anim Dynamics - Bone: dyn_Hippie_hair_04(ComponentPose=@n8, Alpha=1, ExternalForce="0,0,0") {
    angularDampingOverride=0.98
    angularSpringConstant=100
    bAngularSpring=true
    bGravityOverrideInSimSpace=true
    bOverrideAngularDamping=true
    bUseSphericalLimits=true
    numSolverIterationsPostUpdate=4
    numSolverIterationsPreUpdate=16
    alphaBoolBlend {blendOption=Linear}
    boundBone {boneName=dyn_Hippie_hair_04}
    physicsBodyDefinitions [{
      collisionType=CustomSphere
      sphereCollisionRadius=1
      boundBone {boneName=dyn_Hippie_hair_04}
      boxExtents=<20 30 30>
      constraintSetup {angularTargetAxis=AxisZ, angularLimitsMax=<5 20 0>, angularLimitsMin=<-5 -20 -60>, angularTarget=<0 0 1>}
      localJointOffset=<2 0 0>
     }]
    sphericalLimits [{limitRadius=12, drivingBone {boneName=spine_05}, sphereLocalOffset=<0 0 0>}]
   }
   n10 = Anim Dynamics - Bone: dyn_Hippie_hair_01_l(ComponentPose=@n9, Alpha=1, ExternalForce="0,0,0") {
    angularDampingOverride=0.98
    angularSpringConstant=100
    bAngularSpring=true
    bGravityOverrideInSimSpace=true
    bOverrideAngularDamping=true
    numSolverIterationsPostUpdate=4
    numSolverIterationsPreUpdate=16
    alphaBoolBlend {blendOption=Linear}
    boundBone {boneName=dyn_Hippie_hair_01_l}
    physicsBodyDefinitions [{
      collisionType=CustomSphere
      sphereCollisionRadius=1.1
      boundBone {boneName=dyn_Hippie_hair_01_l}
      boxExtents=<20 30 30>
      constraintSetup {angularTargetAxis=AxisZ, angularLimitsMax=<5 20 0>, angularLimitsMin=<-5 -20 -20>, angularTarget=<0 0 1>}
      localJointOffset=<2 0 0>
     }]
   }
   n11 = Anim Dynamics - Bone: dyn_Hippie_hair_02_l(ComponentPose=@n10, Alpha=1, ExternalForce="0,0,0") {
    angularDampingOverride=0.98
    angularSpringConstant=100
    bAngularSpring=true
    bGravityOverrideInSimSpace=true
    bOverrideAngularDamping=true
    bUseSphericalLimits=true
    numSolverIterationsPostUpdate=4
    numSolverIterationsPreUpdate=16
    alphaBoolBlend {blendOption=Linear}
    boundBone {boneName=dyn_Hippie_hair_02_l}
    physicsBodyDefinitions [{
      collisionType=CustomSphere
      sphereCollisionRadius=1.1
      boundBone {boneName=dyn_Hippie_hair_02_l}
      boxExtents=<20 30 30>
      constraintSetup {angularTargetAxis=AxisZ, angularLimitsMax=<5 20 0>, angularLimitsMin=<-5 -20 -20>, angularTarget=<0 0 1>}
      localJointOffset=<2 0 0>
     }]
    sphericalLimits [{limitRadius=12, drivingBone {boneName=spine_05}, sphereLocalOffset=<-2.8 0 -4>}]
   }
   n12 = Anim Dynamics - Bone: dyn_Hippie_hair_01_r(ComponentPose=@n11, Alpha=1, ExternalForce="0,0,0") {
    angularDampingOverride=0.98
    angularSpringConstant=100
    bAngularSpring=true
    bGravityOverrideInSimSpace=true
    bOverrideAngularDamping=true
    numSolverIterationsPostUpdate=4
    numSolverIterationsPreUpdate=16
    alphaBoolBlend {blendOption=Linear}
    boundBone {boneName=dyn_Hippie_hair_01_r}
    physicsBodyDefinitions [{
      collisionType=CustomSphere
      sphereCollisionRadius=1.1
      boundBone {boneName=dyn_Hippie_hair_01_r}
      boxExtents=<20 30 30>
      constraintSetup {angularTargetAxis=AxisZ, angularLimitsMax=<5 20 20>, angularLimitsMin=<-5 -20 0>, angularTarget=<0 0 1>}
      localJointOffset=<-2 0 0>
     }]
   }
   n13 = Anim Dynamics - Bone: dyn_Hippie_hair_02_r(ComponentPose=@n12, Alpha=1, ExternalForce="0,0,0") {
    angularDampingOverride=0.98
    angularSpringConstant=100
    bAngularSpring=true
    bGravityOverrideInSimSpace=true
    bOverrideAngularDamping=true
    bUseSphericalLimits=true
    numSolverIterationsPostUpdate=4
    numSolverIterationsPreUpdate=16
    alphaBoolBlend {blendOption=Linear}
    boundBone {boneName=dyn_Hippie_hair_02_r}
    physicsBodyDefinitions [{
      collisionType=CustomSphere
      sphereCollisionRadius=1.1
      boundBone {boneName=dyn_Hippie_hair_02_r}
      boxExtents=<20 30 30>
      constraintSetup {angularTargetAxis=AxisZ, angularLimitsMax=<5 20 20>, angularLimitsMin=<-5 -20 0>, angularTarget=<0 0 1>}
      localJointOffset=<-2 0 0>
     }]
    sphericalLimits [{limitRadius=12, drivingBone {boneName=spine_05}, sphereLocalOffset=<-1.8 0 4>}]
   }
   n14 = Anim Dynamics - Bone: dyn_Hippie_hair_A_01_l!(ComponentPose=@n13, ExternalForce="0,0,0") -> (Pose) {
    alpha=0.4
    angularDampingOverride=0.99
    angularSpringConstant=500
    bAngularSpring=true
    bGravityOverrideInSimSpace=true
    bOverrideAngularDamping=true
    numSolverIterationsPostUpdate=4
    numSolverIterationsPreUpdate=16
    alphaBoolBlend {blendOption=Linear}
    boundBone {boneName=dyn_Hippie_hair_A_01_l}
    physicsBodyDefinitions [{
      collisionType=CustomSphere
      sphereCollisionRadius=1.1
      boundBone {boneName=dyn_Hippie_hair_A_01_l}
      boxExtents=<30 40 40>
      constraintSetup {angularTargetAxis=AxisZ, angularLimitsMax=<5 20 0>, angularLimitsMin=<-5 -20 -20>, angularTarget=<0 0 1>}
      localJointOffset=<8 0 0>
     }]
   }
   n15 = Anim Dynamics - Bone: dyn_Hippie_hair_A_02_l1!(ComponentPose=@n14, ExternalForce="0,0,0") -> (Pose) {
    alpha=0.4
    angularDampingOverride=0.99
    angularSpringConstant=500
    bAngularSpring=true
    bGravityOverrideInSimSpace=true
    bOverrideAngularDamping=true
    bUseSphericalLimits=true
    numSolverIterationsPostUpdate=4
    numSolverIterationsPreUpdate=16
    alphaBoolBlend {blendOption=Linear}
    boundBone {boneName=dyn_Hippie_hair_A_02_l1}
    physicsBodyDefinitions [{
      collisionType=CustomSphere
      sphereCollisionRadius=1.1
      boundBone {boneName=dyn_Hippie_hair_A_02_l1}
      boxExtents=<30 40 40>
      constraintSetup {angularTargetAxis=AxisZ, angularLimitsMax=<5 20 10>, angularLimitsMin=<-5 -20 -20>, angularTarget=<0 0 1>}
      localJointOffset=<8 0 0>
     }]
    sphericalLimits [
     {limitRadius=20, drivingBone {boneName=spine_05}, sphereLocalOffset=<-6 -9 -5>}
     {limitRadius=4, drivingBone {boneName=neck_02}, sphereLocalOffset=<0 2 0>}
    ]
   }
   n16 = Anim Dynamics - Bone: dyn_Hippie_hair_A_02_l2!(ComponentPose=@n15, ExternalForce="0,0,0") -> (Pose) {
    alpha=0.4
    angularDampingOverride=0.99
    angularSpringConstant=500
    bAngularSpring=true
    bGravityOverrideInSimSpace=true
    bOverrideAngularDamping=true
    bUseSphericalLimits=true
    numSolverIterationsPostUpdate=4
    numSolverIterationsPreUpdate=16
    alphaBoolBlend {blendOption=Linear}
    boundBone {boneName=dyn_Hippie_hair_A_02_l2}
    physicsBodyDefinitions [{
      collisionType=CustomSphere
      sphereCollisionRadius=1.1
      boundBone {boneName=dyn_Hippie_hair_A_02_l2}
      boxExtents=<30 40 40>
      constraintSetup {angularTargetAxis=AxisZ, angularLimitsMax=<5 20 0>, angularLimitsMin=<-5 -20 -20>, angularTarget=<0 0 1>}
      localJointOffset=<8 0 0>
     }]
    sphericalLimits [
     {limitRadius=20, drivingBone {boneName=spine_05}, sphereLocalOffset=<-6 -9 -5>}
     {limitRadius=4, drivingBone {boneName=neck_02}, sphereLocalOffset=<0 2 0>}
    ]
   }
   n17 = "Constraint - dyn_braids_01 (PPPP)"!(
    ComponentPose=@n16
    Alpha=1
    ConstraintWeights_0=0.0007
    ConstraintWeights_1:float=0.809
    ConstraintWeights_2:float=0.1532
    ConstraintWeights_3:float=0.0371
   ) -> (Pose) {
    alphaBoolBlend {blendOption=Linear}
    boneToModify {boneName=dyn_braids_01}
    constraintSetup [
     {targetBone {boneName=dyn_Hippie_hair_02}}
     {targetBone {boneName=head}}
     {targetBone {boneName=dyn_Hippie_hair_01_r}}
     {targetBone {boneName=dyn_Hippie_hair_01}}
    ]
    constraintWeights [1,1,1,1]
   }
   n18 = "Constraint - dyn_braids_02 (PPPP)"!(
    ComponentPose=@n17
    Alpha=1
    ConstraintWeights_0=0.0183
    ConstraintWeights_1:float=0.2882
    ConstraintWeights_2:float=0.5648
    ConstraintWeights_3:float=0.1287
   ) -> (Pose) {
    alphaBoolBlend {blendOption=Linear}
    boneToModify {boneName=dyn_braids_02}
    constraintSetup [
     {targetBone {boneName=dyn_Hippie_hair_02}}
     {targetBone {boneName=head}}
     {targetBone {boneName=dyn_Hippie_hair_01_r}}
     {targetBone {boneName=dyn_Hippie_hair_01}}
    ]
    constraintWeights [1,1,1,1]
   }
   n19 = "Constraint - dyn_braids_03 (PPPP)"!(
    ComponentPose=@n18
    Alpha=1
    ConstraintWeights_0=0.0949
    ConstraintWeights_1:float=0.0919
    ConstraintWeights_2:float=0.7432
    ConstraintWeights_3:float=0.0701
   ) -> (Pose) {
    alphaBoolBlend {blendOption=Linear}
    boneToModify {boneName=dyn_braids_03}
    constraintSetup [
     {targetBone {boneName=dyn_Hippie_hair_02}}
     {targetBone {boneName=head}}
     {targetBone {boneName=dyn_Hippie_hair_01_r}}
     {targetBone {boneName=dyn_Hippie_hair_01}}
    ]
    constraintWeights [1,1,1,1]
   }
   n20 = Anim Dynamics - Bone: dyn_braids_03_a!(ComponentPose=@n19, ExternalForce="0,0,0") -> (Pose) {
    alpha=0.5
    angularDampingOverride=0.99
    bGravityOverrideInSimSpace=true
    lODThreshold=1
    numSolverIterationsPostUpdate=4
    numSolverIterationsPreUpdate=16
    alphaBoolBlend {blendOption=Linear}
    boundBone {boneName=dyn_braids_03_a}
    physicsBodyDefinitions [{
      sphereCollisionRadius=0
      boundBone {boneName=dyn_braids_03_a}
      boxExtents=<40 40 30>
      constraintSetup {angularTargetAxis=AxisZ, twistAxis=AxisZ, angularLimitsMax=<25 45 0>, angularLimitsMin=<0 -45 0>, angularTarget=<0 0 0>}
      localJointOffset=<0 0 -8>
     }]
   }
   n21 = Anim Dynamics - Bone: dyn_braids_03_b!(ComponentPose=@n20, ExternalForce="0,0,0") -> (Pose) {
    alpha=0.5
    angularDampingOverride=0.99
    bGravityOverrideInSimSpace=true
    lODThreshold=1
    numSolverIterationsPostUpdate=4
    numSolverIterationsPreUpdate=16
    alphaBoolBlend {blendOption=Linear}
    boundBone {boneName=dyn_braids_03_b}
    physicsBodyDefinitions [{
      sphereCollisionRadius=0
      boundBone {boneName=dyn_braids_03_b}
      boxExtents=<38 36 32>
      constraintSetup {angularTargetAxis=AxisZ, twistAxis=AxisZ, angularLimitsMax=<25 45 0>, angularLimitsMin=<0 -45 0>, angularTarget=<0 0 0>}
      localJointOffset=<0 0 -7>
     }]
   }
   n22 = Anim Dynamics - Bone: dyn_braids_03_c!(ComponentPose=@n21, ExternalForce="0,0,0") -> (Pose) {
    alpha=0.5
    angularDampingOverride=0.99
    bGravityOverrideInSimSpace=true
    lODThreshold=1
    numSolverIterationsPostUpdate=4
    numSolverIterationsPreUpdate=16
    alphaBoolBlend {blendOption=Linear}
    boundBone {boneName=dyn_braids_03_c}
    physicsBodyDefinitions [{
      sphereCollisionRadius=0
      boundBone {boneName=dyn_braids_03_c}
      boxExtents=<44 41 31>
      constraintSetup {angularTargetAxis=AxisZ, twistAxis=AxisZ, angularLimitsMax=<25 45 0>, angularLimitsMin=<0 -45 0>, angularTarget=<0 0 0>}
      localJointOffset=<0 0 -9>
     }]
   }
   n23 = Anim Dynamics - Bone: dyn_braids_03_d!(ComponentPose=@n22, ExternalForce="0,0,0") -> (Pose) {
    alpha=0.5
    angularDampingOverride=0.99
    bGravityOverrideInSimSpace=true
    lODThreshold=1
    numSolverIterationsPostUpdate=4
    numSolverIterationsPreUpdate=16
    alphaBoolBlend {blendOption=Linear}
    boundBone {boneName=dyn_braids_03_d}
    physicsBodyDefinitions [{
      sphereCollisionRadius=0
      boundBone {boneName=dyn_braids_03_d}
      boxExtents=<46 38 29>
      constraintSetup {angularTargetAxis=AxisZ, twistAxis=AxisZ, angularLimitsMax=<25 45 0>, angularLimitsMin=<0 -45 0>, angularTarget=<0 0 0>}
      localJointOffset=<0 0 -6>
     }]
   }
   Output Pose(Result=Component To Local(ComponentPose=@n23)) {layerGroup=None}
  ]}
 EventGraph {nodes [
   TryGetPawnOwner()
   Event BlueprintUpdateAnimation()
  ]}
]
variables []
